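; Vehicle cosmetic actor for the Plasma Energy boost, driving its NS_Boost Niagara effect.
; When applied, it runs the parent logic and sets the boost's offset by car.
; When a color is ready, it looks up the painted-cosmetics data table row and writes that color to every Niagara parameter listed in Color Names.
; It also passes the color and additional info to SetupPaintSettings.
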(event EventBeginPlay)

(event Collision|EventActorBeginOverlap
  (params (OtherActor :type Actor%20Object%20Reference)))

(event EventTick
  (params (DeltaSeconds :type Float%20%28single-precision%29)))

(event Vehicle|Cosmetics|EventOnCosmeticApplied
  (|CallParentFunction)
  (Class|BPVehicleCosmeticsFunctions|SetBoostOffsetbyCar (Variables|Vehicle|Cosmetics|GetVehicleCosmeticsInterface) (Variables|Default|GetNS_Boost)))

(event Vehicle|Cosmetics|EventOnColorReady
  (params (Color :type Name :default "None") (AdditionalInfo :type Vehicle%20Cosmetics%20Additional%20Parameter%20Infos%20Structure))
  (bind _returnvalue (Utilities|GetDataTableRow "/VehicleCosmetics/DataTables/DT_VehicleCosmetics_Painted.DT_VehicleCosmetics_Painted" Color)
    (:then
      (for _array_element (Variables|Default|GetColor%20Names)
        (Niagara|SetNiagaraVariable(LinearColor) (Variables|Default|GetNS_Boost) _array_element (Utilities|Struct|BreakStructVehicleCosmeticsPainted _returnvalue))))
    (:RowNotFound
      _))
  (CallFunction|SetupPaintSettings Color AdditionalInfo))

(fn ConstructionScript
  (|CallParentFunction))
