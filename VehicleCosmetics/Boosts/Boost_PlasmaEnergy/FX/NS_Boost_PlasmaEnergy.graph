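# Niagara system for the PlasmaEnergy vehicle boost cosmetic: a trail plus nozzle effect at the car's boost position.
# The ribbon emitter spawns rearward trails per port with two renderers, secondary color and width, and source-movement inheritance that fades over life.
# The sprite emitter bursts mesh and sprite particles at that offset, with socket sampling disabled.
# User.bIsPainted switches both emitters to painted color, brightness and material parameters, with extra trail boost in frontend preview.
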NiagaraSystem /VehicleCosmetics/Boosts/Boost_PlasmaEnergy/FX/NS_Boost_PlasmaEnergy
dependencies {
 usedDataInterfaces [&NiagaraDataInterfaceCurve, &NiagaraDataInterfaceSkeletalMesh]
 usedDynamicInputs []
 usedModules [
  &self:SystemSpawnScript.SystemScriptSource.SystemScriptGraph.NiagaraNodeAssignment_0.SetVariables_#1
  &SystemState
 ]
}
emitters {
 NE_VC_Boost_Ribbon_Parent001 {
  emitterUpdateScript [
   &EmitterState()
   &VC_FilteredSocketCount()
   &VC_RateQualityHandler()
   &VC_SpawnRibbon_v1(Number of Ribbons=$Emitter.PortCount, SpawnRate=50)
  ]
  particleSpawnScript [
   &VC_PaintedVariantsOverride(Purple=<rgba 0.2789 0 0.6795 1>, R1 Orange:LinearColor=<rgba 0.6724 0.2346 0 1>, Use R1 Painted Overrides=true)
   &VC_InitializeRibbon_v1(
    Custom Offset=true
    Lifetime=0.6
    Local Vector Offset=&AddVectorToPosition(
     Position=<-90 0 0>
     Vector=&AddVectorToPosition(Position=$User.DTCarOffset, Vector=Linked() {linkedVariable {name=User.SingleBoostPos, type=NiagaraPosition}})
    )
    Offset Amount=1
    Ribbon Width:NiagaraFloat=&Multiply_Float(A=65, B=1.4)
   )
   &Color(
    Color=&MakeCustomLinearColorFromBool(Bool=$User.bIsPainted, True Color=$User.PaintedColor, False Color=<rgba 1 1 1 1>)
    Scale Alpha=&MakeCustomFloatFromBool(Bool=$User.bIsPainted, True Float=1.5, False Float=1)
   )
   &AddVelocity(Velocity=&MakeCustomVectorFromBool(Bool=$User.bIsFrontendPreview, True Vector=<-2000 0 0>, False Vector=<-2000 0 0>)) {module=AddVelocity001}
   /SetVariables(
    Particles.COLOR2:LinearColor=&MakeCustomLinearColorFromBool(Bool=$User.bIsPainted, True Color=$User.PaintedColor, False Color=<rgba 1 1 1 1>)
    Particles.WIDTH_2=180
   ) {isSetParameters=true}
  ]
  particleUpdateScript [
   &ParticleState()
   &InheritSourceMovement(
    Applied Position Scale=&Multiply_VectorByFloat(
     Vector=<1 1 1>
     Float=&FloatFromCurve(FloatCurve=DataInterface() {propertyValues {Curve {keys [<at 0 1 RCIM_Cubic>, {interpMode=RCIM_Cubic, leaveTangent=-3.4326, time=0.7, value=0}]}}})
    )
    ApplyForce=false
    ApplyPosition=true
    ApplyPositionFalloff=false
   ) {module=InheritSourceMovement001}
   &SolveForcesAndVelocity(Clamp Velocity=false, Limit Acceleration=false)
   &VC_UpdateRibbon_v1()
   &DynamicMaterialParameters(
    Index 0 Param 1=&MakeCustomFloatFromBool(Bool=$User.bIsPainted, True Float=5, False Float=0)
    Index 2 Param 1=5
    Module.Index0Param0Write=true
    Module.Index0Param1Write=true
    Module.Index0Param2Write=true
    Module.Index0Param3Write=true
    Module.Index2Param0Write=true
    Module.Index2Param1Write=true
    Module.Index2Param2Write=true
    Module.Index2Param3Write=true
    Write Parameter Index 2=true
   )
   &ScaleRibbonWidth(Ribbon Width Scale=&FloatFromCurve(FloatCurve=DataInterface() {propertyValues {Curve {keys [
         <at 0 1.5 RCIM_Cubic>
         {arriveTangent=-0.293, interpMode=RCIM_Cubic, leaveTangent=0.0883, tangentMode=RCTM_Break, time=0.2, value=0.6}
         <at 1 0.3 RCIM_Cubic>
        ]}}}))
   &ScaleColor(
    Scale RGB=&VectorFromFloat(Value=&MakeCustomFloatFromBool(Bool=$User.bIsPainted, True Float=$User.PaintedScalar, False Float=1))
    ScaleA=true
    ScaleRGB=true
   )
   /SetVariables(
    Particles.INDEX0:Vector4f=&MakeVector4(
     X=&MakeCustomFloatFromBool(Bool=$User.bIsPainted, True Float=5, False Float=0)
     Y=&MakeCustomFloatFromBool(Bool=$User.bIsPainted, True Float=$User.HighlightPowerTrail, False Float=2)
     Z=&MakeCustomFloatFromBool(Bool=$User.bIsPainted, True Float=$User.MultiVariantColorTrail, False Float=1)
     W=&MakeCustomFloatFromBool(Bool=$User.bIsPainted, True Float=$User.WhiteColorTrail, False Float=1)
    )
    Particles.INDEX1:Vector4f=&MakeVector4_FromLinearColor(LinearColor=$User.PaintedColor)
    Particles.INDEX2:Vector4f=&MakeVector4(X=0, Y=0.1, Z=1.5, W=0)
    Particles.INDEX3:Vector4f=&MakeVector4(X=0, Y=0, Z=0, W=0)
   ) {isSetParameters=true}
   /SetVariables(
    Particles.SCALECOLOR2:LinearColor=&VC_DynamicScaleColor(
     Color to Scale:LinearColor=$Particles.COLOR2
     RGB Scale:Vector3f=&VectorFromFloat(Value=&MakeCustomFloatFromBool(
       Bool=$User.bIsPainted
       True Float=&MakeCustomFloatFromBool(
        Bool=$User.bIsFrontendPreview
        True Float=&Multiply_Float(A=$User.MultiVariantColorTrail, B=2)
        False Float=&Multiply_Float(A=$User.MultiVariantColorTrail, B=1)
       )
       False Float=1
      ))
     Alpha Scale:NiagaraFloat=&FloatFromCurve(FloatCurve=DataInterface() {propertyValues {Curve {keys [
          {interpMode=RCIM_Cubic, tangentMode=RCTM_Break, time=0.01, value=0}
          {arriveTangent=3.5161, interpMode=RCIM_Cubic, tangentMode=RCTM_Break, time=0.18, value=1}
          {interpMode=RCIM_Cubic, tangentMode=RCTM_Break, time=1, value=0}
         ]}}})
    )
    Particles.WIDTH_2:NiagaraFloat=&FloatFromCurve(
     FloatCurve=DataInterface() {propertyValues {Curve {keys [
         <at 0 1.5 RCIM_Cubic>
         {arriveTangent=-0.293, interpMode=RCIM_Cubic, leaveTangent=0.0883, tangentMode=RCTM_Break, time=0.2, value=0.6}
         <at 1 0.3 RCIM_Cubic>
        ]}}}
     Scale Curve=180
    )
   ) {isSetParameters=true}
  ]
  rendererClasses [&NiagaraRibbonRendererProperties]
  renderers [
   {rendererClass=&NiagaraRibbonRendererProperties}
   {rendererClass=&NiagaraRibbonRendererProperties, rendererIndex=1}
  ]
 }
 NE_VC_Boost_Sprite_Parent {
  emitterUpdateScript [
   &EmitterState(
    Inactive Response=::NewEnumerator1
    Life Cycle Mode=::NewEnumerator1
    Loop Behavior=::NewEnumerator1
    Loop Duration Mode=::NewEnumerator1
    UseLoopDelay=false
   )
   &VC_FilteredSocketCount() {enabled=false}
   &SpawnBurst_Instantaneous(Use Spawn Probability=false, UseLoopCountLimit=false) {module=SpawnBurst_Instantaneous001}
   &SpawnRate(SpawnRate=50, Use Spawn Probability=false) {enabled=false}
  ]
  particleSpawnScript [
   &VC_PaintedVariantsOverride(Purple=<rgba 0.2789 0 0.6795 1>, Use R1 Painted Overrides=true)
   &InitializeParticle(
    Color Mode=::NewEnumerator0
    Lifetime=$Emitter.CurrentLoopDuration
    Mesh Scale Mode=::NewEnumerator3
    Mesh Uniform Scale=&Multiply_Float(A=0.15, B=2.8)
    Position Offset=&AddVectorToPosition(
     Position=<-110 0 0>
     Vector=&AddVectorToPosition(Position=$User.SingleBoostPos, Vector=Linked() {linkedVariable {name=User.DTCarOffset, type=NiagaraPosition}})
    )
    Position Offset Coordinate Space=::Local
    Sprite Rotation Mode=::NewEnumerator2
    UsePositionOffset=true
    Write Material Random=true
   )
   &PartitionParticles(Number Of Partitions=$Emitter.PortCount) {enabled=false}
   &SkeletalMeshLocation(
    Bone / Socket Index=$Particles.PartitionParticles.Partition
    Bone / Socket Sampling Mode=::NewEnumerator5
    MaskSpawnGroup=false
    Mesh Sampling Type=::NewEnumerator6
    Skeletal Mesh=$System.SkeletalMesh
   ) {enabled=false}
   &AddVelocity(Velocity=<-1500 0 0>) {enabled=false}
   &Color(Color=&MakeCustomLinearColorFromBool(Bool=$User.bIsPainted, True Color=$User.PaintedColor, False Color=<rgba 1 1 1 1>))
   /SetVariables(Particles.COLOR2:LinearColor=&MakeCustomLinearColorFromBool(Bool=$User.bIsPainted, True Color=$User.PaintedColor, False Color=<rgba 1 1 1 1>)) {isSetParameters=true}
  ]
  particleUpdateScript [
   &ParticleState()
   &ScaleMeshSize(Scale Factor=&VectorFromFloat(Value=&FloatFromCurve(
      FloatCurve=DataInterface() {propertyValues {Curve {keys [
          {interpMode=RCIM_Cubic, tangentMode=RCTM_Break, time=0, value=1}
          {interpMode=RCIM_Cubic, tangentMode=RCTM_Break, time=0.5, value=0.85}
          {interpMode=RCIM_Cubic, tangentMode=RCTM_Break, time=1, value=1}
         ]}}}
      Scale Curve=1.2
     )))
   &SolveForcesAndVelocity(Clamp Velocity=false, Limit Acceleration=false)
   &DynamicMaterialParameters(
    Index 0 Param 1=&MakeCustomFloatFromBool(Bool=$User.bIsPainted, True Float=1, False Float=0)
    Index 0 Param 2=0.8
    Index 2 Param 1=&MakeCustomFloatFromBool(Bool=$User.bIsPainted, True Float=1, False Float=0)
    Index 2 Param 2=1
    Module.Index0Param0Write=true
    Module.Index0Param1Write=true
    Module.Index0Param2Write=true
    Module.Index0Param3Write=true
    Module.Index1Param0Write=true
    Module.Index1Param1Write=true
    Module.Index1Param2Write=true
    Module.Index1Param3Write=true
    Module.Index2Param0Write=true
    Module.Index2Param1Write=true
    Module.Index2Param2Write=true
    Module.Index2Param3Write=true
    Module.Index3Param0Write=true
    Module.Index3Param1Write=true
    Module.Index3Param2Write=true
    Module.Index3Param3Write=true
    Write Parameter Index 1=true
    Write Parameter Index 2=true
    Write Parameter Index 3=true
   )
   &ScaleColor(
    Scale RGB=&VectorFromFloat(Value=&MakeCustomFloatFromBool(Bool=$User.bIsPainted, True Float=$User.PaintedScalar, False Float=1))
    ScaleA=true
    ScaleRGB=true
   )
   /SetVariables(Particles.SCALE_COLOR2:LinearColor=&VC_DynamicScaleColor(
     Color to Scale:LinearColor=$Particles.COLOR2
     RGB Scale:Vector3f=&VectorFromFloat(Value=&MakeCustomFloatFromBool(Bool=$User.bIsPainted, True Float=&Multiply_Float(A=$User.MultiVariantColorCone, B=1.2), False Float=1))
     Alpha Scale:NiagaraFloat=1
    )) {isSetParameters=true}
  ]
  rendererClasses [&NiagaraMeshRendererProperties, &NiagaraSpriteRendererProperties]
  renderers [
   {rendererClass=&NiagaraMeshRendererProperties, rendererIndex=0}
   {rendererClass=&NiagaraMeshRendererProperties, rendererIndex=1}
   {rendererClass=&NiagaraSpriteRendererProperties, rendererIndex=2}
  ]
 }
}
summary {
 systemName=NS_Boost_PlasmaEnergy
 emitters [
  {
   emitterName=NE_VC_Boost_Sprite_Parent
   rendererClasses [&NiagaraMeshRendererProperties, &NiagaraSpriteRendererProperties]
  }
  {emitterName=NE_VC_Boost_Ribbon_Parent001, rendererClasses [&NiagaraRibbonRendererProperties]}
 ]
 userVariables [
  User.bIsFrontendPreview {type=NiagaraBool}
  User.bIsPainted {type=NiagaraBool}
  User.ColorVariantID {type=NiagaraInt32}
  User.DTCarOffset {type=NiagaraPosition, description="", defaultValue {}}
  User.HighlightPowerTrail {type=NiagaraFloat, defaultValue=2.5, description=""}
  User.MultiVariantColorCone {type=NiagaraFloat, defaultValue=2.5, description=""}
  User.MultiVariantColorTrail {type=NiagaraFloat, defaultValue=1.5, description=""}
  User.PaintedColor {type=LinearColor, description="", defaultValue {struct=LinearColor, value=<rgba 1 1 1 1>}}
  User.PaintedScalar {type=NiagaraFloat, defaultValue=1, description=""}
  User.PortCount {type=NiagaraInt32, defaultValue=1}
  User.SingleBoostPos {type=NiagaraPosition, description="", defaultValue {}}
  User.WhiteColorTrail {type=NiagaraFloat, defaultValue=1, description=""}
 ]
}
