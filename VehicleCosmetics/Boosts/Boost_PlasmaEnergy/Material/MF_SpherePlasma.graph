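# Material function for the Plasma Energy vehicle boost that produces an animated, glowing plasma look, apparently for spheres.
# A panned ribbon mask offsets the UVs of a second panned sample of the same texture, creating distorted flow shaped by a power input.
# That mask, colored and brightened, multiplies a tinted Fresnel plus an inverted Fresnel, then itself again with Alpha.
# Emissive and the alpha output receive the same result.
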MaterialFunction /VehicleCosmetics/Boosts/Boost_PlasmaEnergy/Material/MF_SpherePlasma
graph {nodes [
  Constant() {r=5}
  n1 = FunctionInput() {bUsePreviewValueAsDefault=true, inputName=1stVelocity, inputType=FunctionInput_Vector2, previewValue=<1 0 0 1>}
  n2 = Panner(Coordinate=TextureCoordinate(), Speed=@n1) {speedX=1}
  n3 = TextureObject() {texture=/DealershipSharedLibrary/FX/Textures/TexturePackage/T_VCVFX_PlasmaRibbonMask02_RGB}
  n4 = Add(A=TextureCoordinate(), B=Multiply(A=TextureSample(UVs=@n2, Tex=@n3).G, B=Constant() {r=0.2}))
  n5 = FunctionInput() {bUsePreviewValueAsDefault=true, inputName=2ndVelocity, inputType=FunctionInput_Vector2, previewValue=<0.2 0.3 0 1>}
  n6 = FunctionInput() {inputName=AdjustmenLayerPower, inputType=FunctionInput_Scalar, sortPriority=3, previewValue=<1 1 1 1>}
  n7 = Power(
   Base=TextureSample(
    UVs=Panner(Coordinate=@n4, Speed=@n5) {speedX=0.2, speedY=0.3}
    Tex=TextureObject() {texture=&T_VCVFX_PlasmaRibbonMask02_RGB}
   ).G
   Exp=@n6
  )
  n8 = Multiply(A=@n7, B=FunctionInput() {bUsePreviewValueAsDefault=true, inputName=Color, previewValue=<1 1 1 1>})
  n9 = Constant() {r=5}
  n10 = FunctionInput() {inputName=Power_1stFresnel, inputType=FunctionInput_Scalar, sortPriority=4, previewValue=<5 1 1 1>}
  n11 = Constant() {r=5}
  n12 = Constant() {r=0.3}
  n13 = Multiply(
   A=MaterialFunctionCall(
    Invert Fresnel=StaticBool()
    Power=@n10
    Cheap contrast bright=MaterialFunctionCall(In=@n11, Contrast=@n12) {materialFunction=&CheapContrast}.Result
   ) {materialFunction=&Fresnel_Function}.Result
   B=Multiply(
    A=FunctionInput() {inputName=ColorFresnel, previewValue=<1 0.6 1 1>}
    B=FunctionInput() {inputName=PowerFrenelColor, inputType=FunctionInput_Scalar, previewValue=<1 1 1 1>}
   )
  )
  n14 = MaterialFunctionCall(
   Invert Fresnel=StaticBool() {value=true}
   Power=Constant() {r=3}
   Cheap contrast dark=MaterialFunctionCall(In=Constant() {r=5}, Contrast=Constant() {r=0.3}) {materialFunction=&CheapContrast}.Result
  ) {materialFunction=&Fresnel_Function}
  n15 = Multiply(
   A=Multiply(A=Multiply(A=@n8, B=@n9), B=Add(A=@n13, B=@n14.Result))
   B=Multiply(A=@n7, B=FunctionInput() {inputName=Alpha, inputType=FunctionInput_Scalar, sortPriority=6, previewValue=<1 1 1 1>})
  )
  FunctionOutput(None=@n15) {outputName=Emissive}
  FunctionOutput(None=@n15) {outputName="Alpha_{"}
  Multiply(A=Power(Base=@n14.Result) {constExponent=5}) {constB=3}
 ]}
