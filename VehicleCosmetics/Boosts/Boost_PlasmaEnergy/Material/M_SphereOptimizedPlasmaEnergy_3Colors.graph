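# Translucent unlit Niagara mesh-particle material for a three-color plasma energy sphere in the Plasma Energy vehicle boost.
# Three MF_SpherePlasma layers combine into a white layer plus base and middle tints, using color parameters as lerp alphas.
# A dynamic-parameter paint flag switches to a desaturated version tinted by particle color, feeding MF_FX_EmissiveAdjustments.
# Two layers form an alpha term, and scrolling UV sine waves, faded at UV edges, build a world-position-offset wobble.
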Material /VehicleCosmetics/Boosts/Boost_PlasmaEnergy/Material/M_SphereOptimizedPlasmaEnergy_3Colors
graph {
 properties {
  bAutomaticallySetUsageInEditor=false
  bUsedWithNiagaraMeshParticles=true
  blendMode=BLEND_Translucent
  pixelDepthOffsetMode=PDOM_Legacy
  shadingModel=MSM_Unlit
  usedShadingModels=Unlit
 }
 nodes [
  n1 = DynamicParameter() {parameterIndex=2, defaultValue=<rgba 0 1 1 1>, paramNames [IsPainted?, MultPowerPainted, Param3, Param4]}
  n2 = Constant3Vector() {constant=<rgba 1 1 1 0>}
  n3 = Constant3Vector() {constant=<rgba 0.1771 0.1771 0.1771 0>}
  n4 = Constant() {r=1}
  n5 = DynamicParameter() {
   parameterIndex=3
   defaultValue=<rgba 0 1 1 1>
   paramNames [WhiteColorSphereR, WhiteColorSphereG, WhiteColorSphereB, WhiteColorSphereMultiply]
  }
  n6 = Multiply(
   A=MaterialFunctionCall!(
    2ndVelocity
    1stVelocity
    Color=@n2
    ColorFresnel=@n3
    PowerFrenelColor=@n4
    AdjustmenLayerPower=Constant() {r=3}
    Power_1stFresnel=Constant() {r=3}
    Alpha=Constant() {r=1}
   ) {materialFunction=&/MF_SpherePlasma}."Alpha_{"
   B=Multiply(A=@n5.WhiteColorSphereMultiply, B=@n5.WhiteColorSphereMultiply)
  )
  n7 = Desaturation(None=@n6)
  n8 = MaterialFunctionCall(
   Color=@n2
   ColorFresnel=@n3
   PowerFrenelColor=@n4
   AdjustmenLayerPower=ScalarParameter() {defaultValue=0.7, parameterName=AdjustmenLayer2}
   Power_1stFresnel=Constant() {r=3}
   Alpha=Constant() {r=1}
  ) {materialFunction=&MF_SpherePlasma}
  n9 = MaterialFunctionCall(
   Color=@n2
   ColorFresnel=@n3
   PowerFrenelColor=@n4
   AdjustmenLayerPower=ScalarParameter() {defaultValue=0.9, parameterName=AdjustmenLayer1}
   Power_1stFresnel=Constant() {r=3}
   Alpha=Constant() {r=0.33}
  ) {materialFunction=&MF_SpherePlasma}
  n10 = LinearInterpolate(B=@n9."Alpha_{", Alpha=VectorParameter() {parameterName=Base_Color_1, defaultValue=<rgba 0.1 0.1 1 1>}.RGB)
  n11 = Desaturation(None=Add(A=@n8."Alpha_{", B=Desaturation(None=@n10)))
  n12 = Multiply(A=Multiply(A=Add(A=@n7, B=@n11), B=ParticleColor().RGB), B=@n1.MultPowerPainted)
  n13 = Add(
   A=@n6
   B=LinearInterpolate(A=@n10, B=@n8."Alpha_{", Alpha=VectorParameter() {parameterName=Middle_Color1, defaultValue=<rgba 0 0.5 1 1>}.RGB)
  )
  MaterialFunctionCall(Emissive=Multiply(
    A=If(A=@n1.IsPainted?, "A > B"=@n12, "A == B"=@n12, "A < B"=@n13) {constB=0.5}
    B=ScalarParameter() {defaultValue=1, parameterName=MultiColorFinal}
   )) {materialFunction=&MF_FX_EmissiveAdjustments}
  Multiply(A=Add(A=@n8."Alpha_{", B=@n9."Alpha_{"), B=ScalarParameter() {defaultValue=2, parameterName=PowerAlpha})
  n14 = Constant()
  n15 = Divide(A=ComponentMask(None=TextureCoordinate()) {r=true}, B=ScalarParameter() {defaultValue=0.5, group=WOP, parameterName=CountWaves2})
  n16 = Time()
  n17 = Multiply(A=ScalarParameter() {defaultValue=0.5, group=WOP, parameterName=HoritzontalSpeed}, B=OneMinus(None=@n16))
  n18 = Multiply(A=ScalarParameter() {defaultValue=-0.2, group=WOP, parameterName=HoritzontalSpeed2}, B=OneMinus(None=@n16))
  n19 = Multiply(A=Sine(None=Add(A=@n15, B=@n17)), B=Sine(None=Add(A=@n15, B=@n18)) {period=1.1})
  n20 = Constant() {r=1.2}
  n21 = Divide(A=ComponentMask(None=TextureCoordinate()) {g=true}, B=ScalarParameter() {defaultValue=0.5, group=WOP, parameterName=CountWaves2})
  n22 = Add(A=@n21, B=Multiply(A=ScalarParameter() {defaultValue=0.5, group=WOP, parameterName=VerticalSpeed}, B=OneMinus(None=@n16)))
  n23 = Add(A=@n21, B=Multiply(A=ScalarParameter() {defaultValue=-0.5, group=WOP, parameterName=VerticalSpeed2}, B=OneMinus(None=@n16)))
  n24 = Multiply(A=OneMinus(None=Sine(None=@n22) {period=1.1}), B=OneMinus(None=Sine(None=@n23)))
  n25 = Constant() {r=1.2}
  n26 = Multiply(
   A=AppendVector(A=Power(Base=@n19, Exp=@n20), B=Power(Base=@n24, Exp=@n25))
   B=ScalarParameter() {defaultValue=10, group=WOP, parameterName=Strenght}
  )
  n27 = ComponentMask(None=TextureCoordinate()) {r=true}
  n28 = OneMinus(None=Power(Base=@n27, Exp=ScalarParameter() {defaultValue=1, group=WOP, parameterName=PowerMiddle2}))
  n29 = ComponentMask(None=TextureCoordinate()) {g=true}
  n30 = OneMinus(None=Power(Base=@n29, Exp=ScalarParameter() {defaultValue=1, group=WOP, parameterName=PowerMiddle2}))
  n31 = Multiply(
   A=Multiply(A=Multiply(A=@n28, B=@n27), B=ScalarParameter() {defaultValue=7, group=WOP, parameterName=MultPower2Horizontal})
   B=Multiply(A=Multiply(A=@n30, B=@n29), B=ScalarParameter() {defaultValue=6, group=WOP, parameterName=MultPower2Vertical})
  )
  Reroute(None=AppendVector(A=AppendVector(A=@n14, B=@n14), B=Multiply(A=@n26, B=@n31)))
 ]
}
