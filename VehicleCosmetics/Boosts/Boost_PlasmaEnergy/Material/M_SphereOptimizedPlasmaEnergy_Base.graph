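# Translucent unlit base material for the plasma energy sphere of a vehicle boost effect, used on Niagara mesh particles.
# Two MF_SpherePlasma layers with different fresnel settings are tinted by darkness and base colors, scaled by a dynamic parameter, then blended.
# The IsPainted particle parameter switches to a desaturated blend tinted by particle color, before MF_FX_EmissiveAdjustments.
# A "WOP" branch builds scrolling sine waves faded at UV edges, ending in a reroute with no visible consumer.
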Material /VehicleCosmetics/Boosts/Boost_PlasmaEnergy/Material/M_SphereOptimizedPlasmaEnergy_Base
graph {
 properties {
  bAutomaticallySetUsageInEditor=false
  bUsedWithNiagaraMeshParticles=true
  blendMode=BLEND_Translucent
  pixelDepthOffsetMode=PDOM_Legacy
  shadingModel=MSM_Unlit
  usedShadingModels=Unlit
 }
 nodes [
  n1 = DynamicParameter() {defaultValue=<rgba 0 1 1 1>, paramNames [IsPainted?, PowerColorDefaultVersion, Param3, Param4]}
  n2 = Constant3Vector() {constant=<rgba 1 1 1 0>}
  n3 = Constant() {r=2.5}
  n4 = Constant() {r=0.4}
  n5 = Constant() {r=3}
  n6 = Constant() {r=1}
  n7 = DynamicParameter() {defaultValue=<rgba 0 1 1 1>, paramNames [IsPainted?, PowerColorDefaultVersion, Param3, Param4]}
  n8 = Multiply(
   A=Saturate(None=MaterialFunctionCall!(2ndVelocity, 1stVelocity, Color=@n2, ColorFresnel=@n2, PowerFrenelColor=@n3, AdjustmenLayerPower=@n4, Power_1stFresnel=@n5, Alpha=@n6) {materialFunction=&/MF_SpherePlasma}."Alpha_{")
   B=Multiply(A=VectorParameter() {parameterName=Darkness_Color, defaultValue=<rgba 0.5 0 1 1>}.RGB, B=@n7.PowerColorDefaultVersion)
  )
  n9 = Multiply(
   A=MaterialFunctionCall(Color=@n2, ColorFresnel=@n2, PowerFrenelColor=@n3, AdjustmenLayerPower=Constant(), Power_1stFresnel=Constant() {r=5}, Alpha=Constant() {r=1}) {materialFunction=&MF_SpherePlasma}."Alpha_{"
   B=Multiply(A=VectorParameter() {parameterName=Base_Color1, defaultValue=<rgba 0.7708 0.2313 0.7708 1>}.RGB, B=@n7.PowerColorDefaultVersion)
  )
  n10 = Multiply(A=LinearInterpolate(A=Desaturation(None=@n8), B=Desaturation(None=@n9)) {constAlpha=0.25}, B=ParticleColor().RGB)
  n11 = LinearInterpolate(A=@n8, B=@n9) {constAlpha=0.25}
  MaterialFunctionCall(Emissive=Multiply(
    A=If(A=@n1.IsPainted?, "A > B"=@n10, "A == B"=@n10, "A < B"=@n11) {constB=0.5}
    B=ScalarParameter() {defaultValue=1, parameterName=MultiColorFinal}
   )) {materialFunction=&MF_FX_EmissiveAdjustments}
  n12 = Constant()
  n13 = Divide(A=ComponentMask(None=TextureCoordinate()) {r=true}, B=ScalarParameter() {defaultValue=0.5, group=WOP, parameterName=CountWaves2})
  n14 = Time()
  n15 = Multiply(A=ScalarParameter() {defaultValue=0.5, group=WOP, parameterName=HoritzontalSpeed}, B=OneMinus(None=@n14))
  n16 = Multiply(A=ScalarParameter() {defaultValue=-0.2, group=WOP, parameterName=HoritzontalSpeed2}, B=OneMinus(None=@n14))
  n17 = Multiply(A=Sine(None=Add(A=@n13, B=@n15)), B=Sine(None=Add(A=@n13, B=@n16)) {period=1.1})
  n18 = Constant() {r=1.2}
  n19 = Divide(A=ComponentMask(None=TextureCoordinate()) {g=true}, B=ScalarParameter() {defaultValue=0.5, group=WOP, parameterName=CountWaves2})
  n20 = Add(A=@n19, B=Multiply(A=ScalarParameter() {defaultValue=0.5, group=WOP, parameterName=VerticalSpeed}, B=OneMinus(None=@n14)))
  n21 = Add(A=@n19, B=Multiply(A=ScalarParameter() {defaultValue=-0.5, group=WOP, parameterName=VerticalSpeed2}, B=OneMinus(None=@n14)))
  n22 = Multiply(A=OneMinus(None=Sine(None=@n20) {period=1.1}), B=OneMinus(None=Sine(None=@n21)))
  n23 = Constant() {r=1.2}
  n24 = Multiply(
   A=AppendVector(A=Power(Base=@n17, Exp=@n18), B=Power(Base=@n22, Exp=@n23))
   B=ScalarParameter() {defaultValue=50, group=WOP, parameterName=Strenght}
  )
  n25 = ComponentMask(None=TextureCoordinate()) {r=true}
  n26 = OneMinus(None=Power(Base=@n25, Exp=ScalarParameter() {defaultValue=1, group=WOP, parameterName=PowerMiddle2}))
  n27 = ComponentMask(None=TextureCoordinate()) {g=true}
  n28 = OneMinus(None=Power(Base=@n27, Exp=ScalarParameter() {defaultValue=1, group=WOP, parameterName=PowerMiddle2}))
  n29 = Multiply(
   A=Multiply(A=Multiply(A=@n26, B=@n25), B=ScalarParameter() {defaultValue=7, group=WOP, parameterName=MultPower2Horizontal})
   B=Multiply(A=Multiply(A=@n28, B=@n27), B=ScalarParameter() {defaultValue=6, group=WOP, parameterName=MultPower2Vertical})
  )
  Reroute(None=AppendVector(A=AppendVector(A=@n12, B=@n12), B=Multiply(A=@n24, B=@n29)))
 ]
}
