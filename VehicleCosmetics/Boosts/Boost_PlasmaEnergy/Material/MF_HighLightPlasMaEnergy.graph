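# Material function for the highlight of the Plasma Energy vehicle boost, outputting Emissive and Alpha.
# It blends three center-scaled UV sizes along U, pans them, and samples two plasma masks whose powered product tints Color for Emissive.
# Alpha multiplies that pattern by a plasma mask, a U-based edge fade, a noise-distorted scrolling ribbon mask and the Alpha input.
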MaterialFunction /VehicleCosmetics/Boosts/Boost_PlasmaEnergy/Material/MF_HighLightPlasMaEnergy
graph {nodes [
  Constant2Vector() {g=0.3, r=1}
  Constant2Vector() {g=0.8, r=1}
  Constant2Vector() {g=1.3, r=1}
  n1 = TextureCoordinate()
  n2 = FunctionInput() {bUsePreviewValueAsDefault=true, inputName=1st Size, inputType=FunctionInput_Vector2, previewValue=<1 1 0 1>}
  n3 = MaterialFunctionCall(UVs=@n1, Texture Scale=@n2) {materialFunction=&ScaleUVsByCenter}
  n4 = FunctionInput() {bUsePreviewValueAsDefault=true, inputName=2nd Size, inputType=FunctionInput_Vector2, previewValue=<1 0.8 0 1>}
  n5 = MaterialFunctionCall(UVs=@n1, Texture Scale=@n4) {materialFunction=&ScaleUVsByCenter}
  n6 = Multiply(A=Power(Base=ComponentMask(None=@n1) {r=true}) {constExponent=0.5}) {constB=2}
  n7 = Saturate(None=@n6)
  n8 = FunctionInput() {bUsePreviewValueAsDefault=true, inputName=3nd Size, inputType=FunctionInput_Vector2, previewValue=<1 0.3 0 1>}
  n9 = Subtract(A=@n6)
  n10 = Panner(Coordinate=LinearInterpolate(
    A=LinearInterpolate(A=@n3.UVs, B=@n5.UVs, Alpha=@n7)
    B=MaterialFunctionCall(UVs=@n1, Texture Scale=@n8) {materialFunction=&ScaleUVsByCenter}.UVs
    Alpha=Saturate(None=@n9)
   )) {speedX=-1}
  n11 = TextureSample(UVs=@n10) {texture=/DealershipSharedLibrary/FX/Textures/TexturePackage/T_VCVFX_PlasmaMask03_RGB}
  n12 = Power(
   Base=Multiply(A=TextureSample(UVs=@n10) {texture=/DealershipSharedLibrary/FX/Textures/TexturePackage/T_VCVFX_PlasmaRibbonMask01_RGB}.R, B=@n11.G)
   Exp=FunctionInput() {inputName=AdjustmentPower, inputType=FunctionInput_Scalar}
  )
  n13 = ComponentMask(None=TextureCoordinate()) {r=true}
  n14 = Multiply(A=@n13, B=OneMinus(None=@n13))
  n15 = FunctionInput() {inputName=FadeOut, inputType=FunctionInput_Scalar}
  n16 = Power(Base=Multiply(A=@n13, B=Power(Base=@n14, Exp=@n15)), Exp=FunctionInput() {inputName=MaskEdges, inputType=FunctionInput_Scalar})
  n17 = Multiply(A=Multiply(A=@n12, B=Multiply(A=@n11.G, B=@n16)), B=Constant() {r=3})
  n18 = Panner(Coordinate=TextureCoordinate() {uTiling=1.5}) {speedX=-0.1, speedY=0.6}
  n19 = Add(
   A=TextureCoordinate()
   B=Multiply(
    A=TextureSample(UVs=@n18) {texture=/DealershipSharedLibrary/FX/Textures/TexturePackage/T_VCVFX_PlasmaRibbonMask02_RGB}.G
    B=Constant() {r=0.1}
   )
  )
  n20 = Power(
   Base=TextureSample(UVs=Panner(Coordinate=@n19) {speedY=-0.5}) {texture=&T_VCVFX_PlasmaRibbonMask01_RGB}.G
   Exp=Constant() {r=0.5}
  )
  FunctionOutput(None=Multiply(A=FunctionInput() {inputName=Alpha, inputType=FunctionInput_Scalar}, B=Multiply(A=@n17, B=@n20))) {outputName=Alpha, sortPriority=1}
  FunctionOutput(None=Multiply(A=FunctionInput() {inputName=Color}, B=@n12)) {outputName=Emissive, sortPriority=1}
  Panner() {speedX=-1}
  ParticleColor()
 ]}
