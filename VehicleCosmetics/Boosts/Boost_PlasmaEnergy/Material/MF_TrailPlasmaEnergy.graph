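# Material function for the plasma energy boost trail, outputting Emissive and Alpha from scrolling ribbon and plasma masks.
# Three size inputs are blended along U through ScaleUVsByCenter to vary the mask UV scale over the trail's length.
# A panned ribbon sample distorts a second one, which is multiplied by plasma and shaped by AdjustmentLayer, then tinted by ColorEmissive.
# Alpha further applies plasma, adjustable U-end fades and an inverted, unpanned, U-mirrored ribbon mask.
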MaterialFunction /VehicleCosmetics/Boosts/Boost_PlasmaEnergy/Material/MF_TrailPlasmaEnergy
graph {
 properties {previewBlendMode=BLEND_Translucent}
 nodes [
  n1 = TextureCoordinate() {uTiling=1.5}
  n2 = FunctionInput() {bUsePreviewValueAsDefault=true, inputName=1st Size, inputType=FunctionInput_Vector2, previewValue=<1 1 0 1>}
  n3 = FunctionInput() {bUsePreviewValueAsDefault=true, inputName=2nd Size, inputType=FunctionInput_Vector2, previewValue=<1 0.8 0 1>}
  n4 = Multiply(A=Power(Base=ComponentMask(None=@n1) {r=true}) {constExponent=0.5}) {constB=2}
  n5 = LinearInterpolate(
   A=MaterialFunctionCall(UVs=@n1, Texture Scale=@n2) {materialFunction=&ScaleUVsByCenter}.UVs
   B=MaterialFunctionCall(UVs=@n1, Texture Scale=@n3) {materialFunction=&ScaleUVsByCenter}.UVs
   Alpha=Saturate(None=@n4)
  )
  n6 = FunctionInput() {bUsePreviewValueAsDefault=true, inputName=3nd Size, inputType=FunctionInput_Vector2, previewValue=<1 0.3 0 1>}
  n7 = MaterialFunctionCall(UVs=@n1, Texture Scale=@n6) {materialFunction=&ScaleUVsByCenter}
  n8 = Saturate(None=Subtract(A=@n4))
  n9 = TextureSample(UVs=Panner(Coordinate=LinearInterpolate(A=@n5, B=@n7.UVs, Alpha=@n8)) {speedX=-0.6}) {texture=/DealershipSharedLibrary/FX/Textures/TexturePackage/T_VCVFX_PlasmaRibbonMask02_RGB}
  n10 = Constant() {r=0.1}
  n11 = Panner(Coordinate=Add(A=TextureCoordinate(), B=Multiply(A=@n9.G, B=@n10))) {speedX=-0.5}
  n12 = TextureCoordinate()
  n13 = Multiply(A=Power(Base=ComponentMask(None=@n12) {r=true}) {constExponent=0.5}) {constB=2}
  n14 = LinearInterpolate(
   A=MaterialFunctionCall(UVs=@n12, Texture Scale=@n2) {materialFunction=&ScaleUVsByCenter}.UVs
   B=MaterialFunctionCall(UVs=@n12, Texture Scale=@n3) {materialFunction=&ScaleUVsByCenter}.UVs
   Alpha=Saturate(None=@n13)
  )
  n15 = MaterialFunctionCall(UVs=@n12, Texture Scale=@n6) {materialFunction=&ScaleUVsByCenter}
  n16 = Saturate(None=Subtract(A=@n13))
  n17 = TextureSample(UVs=Panner(Coordinate=LinearInterpolate(A=@n14, B=@n15.UVs, Alpha=@n16)) {speedX=-1}) {texture=/DealershipSharedLibrary/FX/Textures/TexturePackage/T_VCVFX_PlasmaMask03_RGB}
  n18 = Power(
   Base=Multiply(A=TextureSample(UVs=@n11) {texture=&T_VCVFX_PlasmaRibbonMask02_RGB}.G, B=@n17.G)
   Exp=FunctionInput() {inputName=AdjustmentLayer, inputType=FunctionInput_Scalar}
  )
  n19 = ComponentMask(None=TextureCoordinate()) {r=true}
  n20 = OneMinus(None=@n19)
  n21 = FunctionInput() {inputName=FadeOut, inputType=FunctionInput_Scalar, sortPriority=2}
  n22 = Power(
   Base=Multiply(A=@n19, B=Power(Base=@n20, Exp=@n21))
   Exp=FunctionInput() {inputName=MaskEdges, inputType=FunctionInput_Scalar, sortPriority=1}
  )
  n23 = Multiply(A=Multiply(A=@n18, B=Multiply(A=@n17.G, B=@n22)), B=Constant() {r=4})
  n24 = TextureCoordinate() {uTiling=-1}
  n25 = Multiply(A=Power(Base=ComponentMask(None=@n24) {r=true}) {constExponent=0.5}) {constB=2}
  n26 = LinearInterpolate(
   A=LinearInterpolate(
    A=MaterialFunctionCall(UVs=@n24, Texture Scale=@n2) {materialFunction=&ScaleUVsByCenter}.UVs
    B=MaterialFunctionCall(UVs=@n24, Texture Scale=@n3) {materialFunction=&ScaleUVsByCenter}.UVs
    Alpha=Saturate(None=@n25)
   )
   B=MaterialFunctionCall(UVs=@n24, Texture Scale=@n6) {materialFunction=&ScaleUVsByCenter}.UVs
   Alpha=Saturate(None=Subtract(A=@n25))
  )
  n27 = OneMinus(None=Power(
    Base=TextureSample(UVs=@n26) {texture=&T_VCVFX_PlasmaRibbonMask02_RGB}.B
    Exp=Constant() {r=0.5}
   ))
  FunctionOutput(None=Multiply(A=FunctionInput() {inputName=Alpha, inputType=FunctionInput_Scalar, sortPriority=4}, B=Multiply(A=@n23, B=@n27))) {outputName=Alpha, sortPriority=1}
  FunctionOutput(None=Multiply(A=FunctionInput() {inputName=ColorEmissive, sortPriority=3}, B=@n18)) {outputName=Emissive}
 ]
}
