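# Unlit translucent Niagara ribbon material for the trail of the Plasma Energy vehicle boost.
# The ComplementaryTrail static switch selects either four layered MF_TrailPlasmaEnergy masks with lerped colors plus a highlight, or one base trail.
# Dynamic parameters drive intensities and IsPainted, which swaps default colors for particle-color tinting and a custom highlight color.
# Opacity combines masks, particle alpha, U-based end fades and depth fade, while a time-driven sine builds a Z wave offset.
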Material /VehicleCosmetics/Boosts/Boost_PlasmaEnergy/Material/M_TrailOptimized
graph {
 properties {
  bAutomaticallySetUsageInEditor=false
  bUsedWithNiagaraRibbons=true
  blendMode=BLEND_Translucent
  pixelDepthOffsetMode=PDOM_Legacy
  shadingModel=MSM_Unlit
  usedShadingModels=Unlit
 }
 nodes [
  n1 = Constant()
  n2 = ComponentMask(None=TextureCoordinate() {coordinateIndex=1}) {r=true}
  n3 = ScalarParameter() {defaultValue=0.5, parameterName=CountWaves2, sliderMax=1}
  n4 = ScalarParameter() {defaultValue=0.5, parameterName=Speed}
  n5 = OneMinus(None=Time())
  n6 = Sine(None=Add(A=Divide(A=@n2, B=@n3), B=Multiply(A=@n4, B=@n5)))
  n7 = ScalarParameter() {defaultValue=5, parameterName=Strenght}
  n8 = Power(Base=ComponentMask(None=TextureCoordinate()) {r=true}, Exp=ScalarParameter() {defaultValue=1, parameterName=PowerMiddle2})
  n9 = ScalarParameter() {defaultValue=0.5, parameterName=MultPower}
  AppendVector(A=AppendVector(A=@n1, B=@n1), B=Multiply(A=Multiply(A=@n6, B=@n7), B=Multiply(A=@n8, B=@n9)))
  DepthFade()
  n10 = DynamicParameter() {defaultValue=<rgba 0 1 1 1>, paramNames [IsPainted?, HighlightPower, PowerTrail, PowerColorTrileUp]}
  n11 = Multiply(
   A=MaterialFunctionCall!(
    MaskEdges=Constant() {r=0.5}
    FadeOut=Constant() {r=3}
    Color=Constant() {r=1}
    Alpha=Constant() {r=1}
    1st Size=VectorParameter() {parameterName=Size 1 Highlight, defaultValue=<rgba 1 0.5 0 1>}.RGB
    2nd Size=VectorParameter() {parameterName=Size 2 Highlight, defaultValue=<rgba 1 0.5 0 1>}.RGB
    3nd Size=VectorParameter() {parameterName=Size 3 Highlight, defaultValue=<rgba 1 0.22 0 1>}.RGB
    AdjustmentPower=Constant() {r=0.15}
   ) {materialFunction=&/MF_HighLightPlasMaEnergy}.Emissive
   B=@n10.HighlightPower
  )
  n12 = DynamicParameter() {parameterIndex=1, defaultValue=<rgba 0 1 1 1>, paramNames [R,G,B,MultiColor]}
  n13 = Multiply(A=@n12.RGB, B=@n12.RGB)
  n14 = VectorParameter() {parameterName=ColorHighlightTrail, defaultValue=<rgba 0.6 0.102 1 1>}
  n15 = Multiply(A=@n11, B=Multiply(A=If(A=@n10.IsPainted?, "A > B"=@n13, "A == B"=@n13, "A < B"=@n14.RGB) {constB=0.5}) {constB=1.5})
  n16 = VectorParameter() {parameterName=Trail Size 1 2, defaultValue=<rgba 1 0.4 0 1>}
  n17 = VectorParameter() {parameterName=Trail Size 3, defaultValue=<rgba 1 0.2 0 1>}
  n18 = Multiply(
   A=MaterialFunctionCall!(
    2nd Size=@n16.RGB
    1st Size=@n16.RGB
    3nd Size=@n17.RGB
    AdjustmentLayer=ScalarParameter() {defaultValue=8, parameterName=AdjusmentLayerComplementaryRibbon4}
    MaskEdges=Constant() {r=0.5}
    FadeOut=Constant() {r=3}
    ColorEmissive
    Alpha=Constant() {r=1}
   ) {materialFunction=&/MF_TrailPlasmaEnergy}.Alpha
   B=DynamicParameter() {defaultValue=<rgba 0 1 1 1>, paramNames [IsPainted?, HighlightPower, PowerTrail, PowerColorTrileUp]}.PowerColorTrileUp
  )
  n19 = MaterialFunctionCall(
   2nd Size=@n16.RGB
   1st Size=@n16.RGB
   3nd Size=@n17.RGB
   AdjustmentLayer=ScalarParameter() {defaultValue=2.5, parameterName=AdjusmentLayerComplementaryRibbon3}
   MaskEdges=Constant() {r=0.5}
   FadeOut=Constant() {r=1}
   ColorEmissive=Constant() {r=1}
   Alpha=Constant() {r=1}
  ) {materialFunction=&MF_TrailPlasmaEnergy}
  n20 = DynamicParameter() {
   parameterIndex=2
   defaultValue=<rgba 0 1 1 1>
   paramNames [PowerColorDefaultVersion, 1er Trail Azul AdjustmentLayer, 2do Trail Azul AdjustmentLayer, ""]
  }
  n21 = MaterialFunctionCall(
   2nd Size=@n16.RGB
   1st Size=@n16.RGB
   3nd Size=@n17.RGB
   AdjustmentLayer=@n20.2do Trail Azul AdjustmentLayer
   MaskEdges=Constant() {r=0.5}
   FadeOut=Constant() {r=1}
   ColorEmissive=Constant() {r=1}
   Alpha=Constant() {r=1}
  ) {materialFunction=&MF_TrailPlasmaEnergy}
  n22 = MaterialFunctionCall(
   2nd Size=@n16.RGB
   1st Size=@n16.RGB
   3nd Size=@n17.RGB
   AdjustmentLayer=@n20.1er Trail Azul AdjustmentLayer
   MaskEdges=Constant() {r=0.5}
   FadeOut=Constant() {r=1}
   Alpha=Constant() {r=1}
  ) {materialFunction=&MF_TrailPlasmaEnergy}
  n23 = Add(A=@n18, B=Add(A=@n19.Alpha, B=Add(A=@n21.Alpha, B=@n22.Alpha)))
  n24 = ParticleColor()
  n25 = Add(A=@n15, B=Multiply(A=@n10.PowerTrail, B=Multiply(A=@n23, B=@n24.RGB))) {constB=0}
  n26 = LinearInterpolate(B=VectorParameter() {parameterName=Complementary_Color1, defaultValue=<rgba 0.502 0 1 1>}.RGB, Alpha=@n22.Alpha)
  n27 = VectorParameter() {parameterName=Complementary_Color2, defaultValue=<rgba 0.102 0.102 1 1>}
  n28 = LinearInterpolate(
   A=LinearInterpolate(
    A=LinearInterpolate(A=@n26, B=@n27.RGB, Alpha=@n21.Alpha)
    B=VectorParameter() {parameterName=Complementary_Color3, defaultValue=<rgba 0 0.498 1 1>}.RGB
    Alpha=@n19.Alpha
   )
   B=VectorParameter() {parameterName=Complementary_Color4, defaultValue=<rgba 0.75 0.75 0.75 1>}.RGB
   Alpha=@n18
  )
  n29 = ScalarParameter() {defaultValue=1, parameterName=PowerBlueTrail}
  n30 = If(A=@n10.IsPainted?, "A > B"=@n25, "A == B"=@n25, "A < B"=Add(A=@n15, B=Power(Base=@n28, Exp=@n29) {constExponent=1})) {constB=0.5}
  n31 = Constant() {r=0.5}
  n32 = MaterialFunctionCall(AdjustmentLayer=@n31, MaskEdges=@n31, FadeOut=Constant() {r=3}, Alpha=Constant() {r=1}) {materialFunction=&MF_TrailPlasmaEnergy}
  n33 = LinearInterpolate(
   B=Multiply(A=VectorParameter() {parameterName=BaseColorTrail, defaultValue=<rgba 0.8 0.102 0.8 1>}.RGB, B=@n20.PowerColorDefaultVersion) {constB=10}
   Alpha=@n32.Alpha
  )
  n34 = ParticleColor()
  n35 = Multiply(A=Desaturation(None=@n33), B=@n34.RGB)
  n36 = If(
   A=DynamicParameter() {defaultValue=<rgba 0 1 1 1>, paramNames [IsPainted?, HighlightPower, PowerTrail, PowerColorTrileUp]}.IsPainted?
   "A > B"=@n35
   "A == B"=@n35
   "A < B"=@n33
  ) {constB=0.5}
  MaterialFunctionCall(Emissive=Multiply(
    A=StaticSwitchParameter(True=@n30, False=@n36) {parameterName=ComplementaryTrail?}
    B=ScalarParameter() {defaultValue=1, parameterName=MultiColorFinal}
   )) {materialFunction=&MF_FX_EmissiveAdjustments}
  n37 = Add(A=@n18, B=Add(A=@n19.Alpha, B=Add(A=@n21.Alpha, B=@n22.Alpha)))
  n38 = TextureCoordinate()
  n39 = TextureCoordinate()
  n40 = Multiply(A=OneMinus(None=ComponentMask(None=@n38) {r=true}), B=Power(Base=ComponentMask(None=@n39) {r=true}) {constExponent=0.5})
  n41 = Multiply(A=Multiply(A=Add(A=@n11, B=@n37), B=@n24.A), B=@n40)
  n42 = Multiply(A=@n40, B=Multiply(A=@n34.A, B=@n32.Alpha))
  Multiply(
   A=Multiply(A=ParticleColor().A, B=StaticSwitchParameter(True=@n41, False=@n42) {parameterName=ComplementaryTrail?})
   B=MaterialFunctionCall(Opacity In=Constant() {r=1}, DepthFade Distance=Constant() {r=100}) {materialFunction=&MF_VCVFX_DepthFade}.Result
  )
 ]
}
